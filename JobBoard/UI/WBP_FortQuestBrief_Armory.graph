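# Armory variant of the FortQuestBrief conversation UI widget, used in the Job Board to brief the player on picking a weapon.
# It adds no widgets of its own and inherits its layout from the parent FortQuestBrief class.
# On PreConstruct it calls the parent, then sets the warning text to say that picking a weapon removes it from the terminal.
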WidgetBlueprint /JobBoard/UI/WBP_FortQuestBrief_Armory
graphs [EventGraph {nodes [
   n1 = Event PreConstruct()
   > Parent: PreConstruct(IsDesignTime=@n1.IsDesignTime)
   > SetWarningText!(execute, self:Fort Quest Brief Object Reference, WarningText:Text="LOCTEXT(\"Picking a weapon will remove it from this terminal\")") -> (then)
  ]}]
variables [
 IsConstructed {type=boolean}
 NewCoordinate {type=Vector2D, default=<0 0>}
 PrevCoordinate {type=Vector2D, default=<0 0>}
]
widgets {
 info {
  inheritedWidgetCount=0
  parentClass=/FortniteConversationUI/Conversation/FortQuestBrief.FortQuestBrief_C
  rootWidgetClass=null
  widgetCount=0
 }
 widgets []
}
